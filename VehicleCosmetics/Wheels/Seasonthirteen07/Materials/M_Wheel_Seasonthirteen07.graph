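# Masked master material for the Seasonthirteen07 vehicle wheel, adding custom-colored rim and tire glow to MF_WheelParent_Simple.
# Static switches select GrandChampion, Diamond or Gold looks, toggle rim and tire lights, and swap masks for LOD 2-5.
# GrandChampion animates tire glow with time-driven sine waves, plus opposing line-texture panners in its LOD branch.
# The emissive overrides the parent attributes, which then pass through vehicle damage and de-rez transition functions.
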Material /VehicleCosmetics/Wheels/Seasonthirteen07/Materials/M_Wheel_Seasonthirteen07
graph {
 properties {
  bAutomaticallySetUsageInEditor=false
  bUseMaterialAttributes=true
  bUsedWithMeshDeformer=true
  bUsedWithSkeletalMesh=true
  blendMode=BLEND_Masked
  pixelDepthOffsetMode=PDOM_Legacy
  previewMesh=/VehicleCosmetics/Wheels/Seasonthirteen06/Meshes/SK_Wheel_Seasonthirteen06
  usedShadingModels=Default Lit
 }
 nodes [
  n1 = TextureSampleParameter2D() {group=S13, parameterName=RIM_E, texture=/VehicleCosmetics/Wheels/Seasonthirteen07/Textures/T_Seasonthirteen06_Rim_M_E}
  n2 = OneMinus(None=@n1.R)
  n3 = TextureSample() {samplerType=SAMPLERTYPE_LinearColor, texture=/VehicleCosmetics/Wheels/Seasonthirteen07/Textures/T_Seasonthirteen06_Rim_M_LOD}
  Add(A=@n2, B=@n3.B)
  Add(A=@n2, B=@n1.B)
  n4 = MaterialFunctionCall() {materialFunction=&MF_WheelParent_Simple}
  n5 = NamedRerouteUsage()
  MaterialFunctionCall(In=MaterialFunctionCall(Material=SetMaterialAttributes(MaterialAttributes=@n4.MaterialAttributes, Emissive Color=@n5) {attributeSetTypes [#1]}) {materialFunction=&MF_BaseVehicle_Damage}.Result) {materialFunction=&MF_VehicleDeRez_Transition}
  Multiply(A=@n3.R)
  n6 = TextureSampleParameter2D() {group=S13, parameterName=M_LOD, texture=/VehicleCosmetics/Wheels/Seasonthirteen07/Textures/T_Seasonthirteen06_Tire_M_E_LOD}
  n7 = TextureSampleParameter2D() {group=S13, parameterName=M, texture=/VehicleCosmetics/Wheels/Seasonthirteen07/Textures/T_seasonThirteen07_Tire_M_E}
  NamedRerouteDeclaration(None=StaticSwitchParameter(True=@n6.R, False=@n7.R) {desc=TURN ON FOR LOD 2-5, group=S13, parameterName=LOD2-5, sortPriority=1}) {name=ShaderM_R, nodeColor=<rgba 1 0 0.1282 1>}
  NamedRerouteDeclaration(None=StaticSwitchParameter(True=@n6.G, False=@n7.G) {desc=TURN ON FOR LOD 2-5, group=S13, parameterName=LOD2-5, sortPriority=1}) {name=ShaderM_G, nodeColor=<rgba 0.212 1 0 1>}
  NamedRerouteDeclaration(None=StaticSwitchParameter(True=@n6.B, False=@n7.B) {desc=TURN ON FOR LOD 2-5, group=S13, parameterName=LOD2-5, sortPriority=1}) {name=ShaderM_B, nodeColor=<rgba 0.0104 0 1 1>}
  NamedRerouteDeclaration(None=StaticSwitchParameter(True=@n6.A, False=@n7.A) {desc=TURN ON FOR LOD 2-5, group=S13, parameterName=LOD2-5, sortPriority=1}) {name=ShaderM_A, nodeColor=<rgba 1 0.9588 0.9684 1>}
  NamedRerouteDeclaration(None=VertexColor().R) {name=VC_R, nodeColor=<rgba 1 0 0.1157 1>}
  n8 = VectorParameter() {group=S13, parameterName=CustomColor, defaultValue=<rgba 0.0042 0.5222 1 1>}
  n9 = Multiply(A=@n8.RGB, B=SquareRoot(None=@n8.RGB))
  n10 = ScalarParameter() {defaultValue=3, group=S13, parameterName=EmissiveColorIntensity}
  n11 = Desaturation(None=@n9, Fraction=@n10)
  n12 = StaticSwitchParameter(True=@n11, False=Desaturation(None=SquareRoot(None=@n9), Fraction=@n10)) {parameterName=Diamond}
  n13 = Constant()
  n14 = StaticSwitchParameter(
   True=Multiply(
    A=StaticSwitchParameter(True=@n1.B, False=@n1.B) {group=S13, parameterName=LOds_Emissive_For_RimE}
    B=ScalarParameter() {defaultValue=1, group=S13, parameterName=RimEmissiveGlow}
   )
   False=@n13
  ) {defaultValue=true, parameterName=RimLights}
  n15 = NamedRerouteUsage()
  n16 = StaticSwitchParameter(True=TextureCoordinate() {vTiling=0.05}, False=TextureCoordinate() {vTiling=0.25}) {desc=TURN ON FOR LOD 2-5, group=S13, parameterName=LOD2-5, sortPriority=1}
  n17 = AppendVector(A=Constant() {r=0.5}, B=Constant() {r=1})
  n18 = ComponentMask(None=Frac(None=Multiply(A=@n16, B=@n17))) {g=true}
  n19 = Time()
  n20 = ScalarParameter() {defaultValue=0.5, group=S13, parameterName=GlowPanSpeed}
  n21 = Add(A=@n18, B=Multiply(A=@n19, B=@n20))
  n22 = Clamp(None=ConstantBiasScale(None=Sine(None=@n21) {period=0.5}) {bias=-0.1, scale=2})
  n23 = NamedRerouteUsage()
  n24 = ConstantBiasScale(None=ComponentMask(None=TextureCoordinate()) {r=true}) {bias=-0.5, scale=100}
  n25 = Multiply(A=@n20, B=Constant() {r=0.5})
  n26 = Sine(None=Add(A=@n18, B=Multiply(A=@n19, B=@n25))) {period=0.5}
  n27 = LinearInterpolate(
   A=Clamp(None=ConstantBiasScale(None=@n26) {bias=-0.1, scale=2})
   B=@n22
   Alpha=StaticSwitchParameter(True=OneMinus(None=@n23), False=Clamp(None=@n24)) {desc=TURN ON FOR LOD 2-5, group=S13, parameterName=LOD2-5, sortPriority=1}
  )
  n28 = Add(A=Multiply(A=@n15, B=Constant() {r=1.5}), B=Multiply(A=Multiply(A=@n15, B=@n27), B=Constant() {r=5}))
  n29 = Multiply(A=NamedRerouteUsage(), B=Constant() {r=1.5})
  n30 = ComponentMask(None=TextureCoordinate()) {r=true}
  n31 = Multiply(A=NamedRerouteUsage(), B=OneMinus(None=ConstantBiasScale(None=@n30) {bias=-0.5, scale=100}))
  n32 = TextureCoordinate()
  n33 = TextureSample(UVs=Panner(Coordinate=@n32) {speedX=-0.5}) {texture=/VehicleCosmetics/Wheels/Seasonthirteen07/Textures/T_SeasonThirteen07_Line}
  n34 = TextureSample(UVs=Panner(Coordinate=@n32) {speedX=0.5}) {texture=&T_SeasonThirteen07_Line}
  n35 = TextureCoordinate()
  n36 = Multiply(A=NamedRerouteUsage(), B=ConstantBiasScale(None=ComponentMask(None=@n35) {r=true}) {bias=-0.5, scale=100})
  n37 = Add(A=Multiply(A=Multiply(A=@n33.R, B=@n31), B=Constant() {r=5}), B=Multiply(A=Multiply(A=@n34.R, B=@n36), B=Constant() {r=5}))
  n38 = TextureCoordinate() {uTiling=0.25}
  n39 = AppendVector(A=Constant() {r=0.5}, B=Constant() {r=1})
  n40 = ComponentMask(None=Frac(None=Multiply(A=@n38, B=@n39))) {r=true}
  n41 = Multiply(A=Time(), B=Multiply(A=ScalarParameter() {defaultValue=0.5, group=S13, parameterName=GlowPanSpeed}, B=Constant() {r=0.25}))
  n42 = ConstantBiasScale(None=Sine(None=Add(A=@n40, B=@n41)) {period=0.5}) {bias=-0.1, scale=0.8}
  n43 = Multiply(A=Multiply(A=NamedRerouteUsage(), B=Clamp(None=@n42)), B=Constant() {r=5})
  n44 = Multiply(A=NamedRerouteUsage(), B=@n22)
  n45 = Constant() {r=5}
  n46 = Add(
   A=Add(
    A=StaticSwitchParameter(True=@n37, False=@n43) {desc=TURN ON FOR LOD 2-5, group=S13, parameterName=LOD2-5, sortPriority=1}
    B=Multiply(A=@n44, B=@n45)
   )
   B=StaticSwitchParameter(True=Add(A=@n28, B=@n29), False=@n28) {desc=TURN ON FOR LOD 2-5, group=S13, parameterName=LOD2-5, sortPriority=1}
  )
  n47 = NamedRerouteUsage()
  n48 = NamedRerouteUsage()
  n49 = NamedRerouteUsage()
  n50 = NamedRerouteUsage()
  n51 = Add(A=NamedRerouteUsage(), B=@n50)
  n52 = Constant() {r=5}
  n53 = StaticSwitchParameter(
   True=Multiply(A=Add(A=@n47, B=@n48), B=Constant() {r=3.5})
   False=StaticSwitchParameter(True=Multiply(A=@n49, B=@n50), False=Multiply(A=@n51, B=@n52)) {parameterName=Gold}
  ) {parameterName=Diamond}
  n54 = StaticSwitchParameter(
   True=Multiply(
    A=ScalarParameter() {defaultValue=1, group=S13, parameterName=TireEmissiveGlow}
    B=StaticSwitchParameter(True=@n46, False=@n53) {parameterName=GrandChampion}
   )
   False=@n13
  ) {defaultValue=true, parameterName=TireLights}
  n55 = NamedRerouteUsage()
  NamedRerouteDeclaration(None=Multiply(A=StaticSwitchParameter(True=@n11, False=@n12) {parameterName=GrandChampion}, B=LinearInterpolate(A=@n14, B=@n54, Alpha=@n55))) {name=Emissive, nodeColor=<rgba 1 0 0.8731 1>}
 ]
}
